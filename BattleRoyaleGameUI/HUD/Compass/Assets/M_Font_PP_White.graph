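# UI-domain translucent material for white font in the battle royale HUD compass.
# It reads the White color from the shared MPC_BRGameUI parameter collection and masks it to RGB, so the tint comes from a central palette.
# A FontOpacity scalar parameter, slightly below fully opaque by default, is exposed for transparency.
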Material /BattleRoyaleGameUI/HUD/Compass/Assets/M_Font_PP_White
graph {
 properties {blendMode=BLEND_Translucent, materialDomain=MD_UI, pixelDepthOffsetMode=PDOM_Legacy, usedShadingModels=Default Lit}
 nodes [
  ComponentMask(None=CollectionParameter() {collection=/BattleRoyaleGameUI/Shared/MPC_BRGameUI, parameterName=White}) {b=true, g=true, r=true}
  ScalarParameter() {defaultValue=0.95, parameterName=FontOpacity}
 ]
}
